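# Niagara system for the RetroTech vehicle boost cosmetic, trailing ribbons and sprites from filtered skeletal mesh sockets.
# Two ribbon emitters stream backward with decaying inherited motion; one pulses width, alpha and material through waveforms.
# A GPU sprite emitter packs three spawn groups with distinct velocities and lifetimes.
# Each emitter drives multiple renderers, some through custom per-layer bindings; user parameters control paint color, car offset and frontend-preview tuning.
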NiagaraSystem /VehicleCosmetics/Boosts/Boost_RetroTech/FX/NS_Boost_Retrotech_Opt
dependencies {
 usedDataInterfaces [
  &NiagaraDataInterfaceCurve
  &NiagaraDataInterfaceArrayFloat
  &NiagaraDataInterfaceArrayInt32
  &NiagaraDataInterfaceSpriteRendererInfo
  &NiagaraDataInterfaceSkeletalMesh
 ]
 usedDynamicInputs []
 usedModules [
  &self:SystemSpawnScript.SystemScriptSource.SystemScriptGraph.NiagaraNodeAssignment_0.SetVariables_#1
  &SystemState
 ]
}
emitters {
 NE_VC_Boost_Ribbon {
  emitterUpdateScript [
   &EmitterState()
   &VC_FilteredSocketCount()
   &VC_RateQualityHandler()
   &VC_SpawnRibbon_v1(Number of Ribbons=$Emitter.PortCount, SpawnRate=60)
  ]
  particleSpawnScript [
   &VC_PaintedVariantsOverride() {enabled=false}
   &VC_InitializeRibbon_v1(
    Custom Offset=true
    Enable Skeletal Position=true
    Lifetime=1.1
    Local Vector Offset=&AddVectorToPosition(Position=<0 0 0>, Vector=$User.DTCarOffset)
    Offset Amount=1
   )
   /SetVariables(
    Particles.Gradient_Rib_Width:NiagaraFloat=&Multiply_Float(A=32, B=1)
    Particles.Grid_Rib_Width:NiagaraFloat=&Multiply_Float(A=25.8923, B=1.3)
    Particles.Lightning_Rib_Width:NiagaraFloat=&Multiply_Float(A=39.5, B=1.3)
   ) {isSetParameters=true}
   &AddVelocity(Velocity=<-1700 0 0>) {module=AddVelocity001}
   &Color(Color=&MakeCustomLinearColorFromBool(Bool=$User.bIsPainted, True Color=$User.PaintedColor, False Color=<rgba 25 5 0.2941 1>))
   /VehicleCosmetics/FX/Niagara/Modules/MergeEmitters/VC_ExtraBindingColors(
    ExtraColors=3
    R1 Color:LinearColor=&MakeCustomLinearColorFromBool(Bool=$User.bIsPainted, True Color=$User.PaintedColor, False Color=<rgba 1.2 0.0706 12 2>)
    R2 Color:LinearColor=&MakeCustomLinearColorFromBool(Bool=$User.bIsPainted, True Color=$User.PaintedColor, False Color=<rgba 0.0588 1 5 1>)
    R3 Color:LinearColor=&MakeCustomLinearColorFromBool(Bool=$User.bIsPainted, True Color=$User.PaintedColor, False Color=<rgba 0.2 10 20 1>)
   )
  ]
  particleUpdateScript [
   &ParticleState()
   &InheritSourceMovement(
    Applied Position Scale=&Multiply_VectorByFloat(
     Vector=<1 1 1>
     Float=&FloatFromCurve(FloatCurve=DataInterface() {propertyValues {Curve {keys [
          {arriveTangent=0.0201, interpMode=RCIM_Cubic, leaveTangent=0.0201, tangentMode=RCTM_User, time=0, value=1}
          {arriveTangentWeight=0.07, interpMode=RCIM_Cubic, leaveTangentWeight=0.2633, tangentMode=RCTM_User, time=0.2, value=1}
          {arriveTangent=0.0759, arriveTangentWeight=0.2633, interpMode=RCIM_Cubic, leaveTangent=0.0759, tangentMode=RCTM_User, time=1, value=0}
         ]}}})
    )
    ApplyForce=false
    ApplyPosition=true
    ApplyPositionFalloff=false
   )
   &SolveForcesAndVelocity(Clamp Velocity=false, Limit Acceleration=false)
   &VC_UpdateRibbon_v1(Enable Skeletal Position=true)
   /SetVariables(Particles.Lightning_Rib_Dynamic:Vector4f=&MakeVector4(
     X=&RandomRangeFloat(Minimum=1, Maximum=5)
     Y=0.2
     Z=&RandomRangeFloat(Minimum=5, Maximum=10)
     W=&RandomRangeFloat(Minimum=10, Maximum=15)
    )) {isSetParameters=true}
  ]
  rendererClasses [&NiagaraRibbonRendererProperties]
  renderers [
   {rendererClass=&NiagaraRibbonRendererProperties}
   {rendererClass=&NiagaraRibbonRendererProperties, rendererIndex=1}
   {rendererClass=&NiagaraRibbonRendererProperties, rendererIndex=2}
  ]
 }
 NE_VC_Boost_Sprite {
  simTarget=GPUComputeSim
  emitterUpdateScript [
   &EmitterState()
   &VC_FilteredSocketCount()
   &SpawnRate(
    SpawnRate=&MakeCustomFloatFromBool(Bool=$User.bIsFrontendPreview, True Float=19, False Float=25)
    Use Spawn Probability=false
   )
   &SpawnRate(
    SpawnGroup=1
    SpawnRate=&MakeCustomFloatFromBool(Bool=$User.bIsFrontendPreview, True Float=15, False Float=20)
    Use Spawn Probability=false
   ) {module=SpawnRate001}
   &SpawnRate(
    SpawnGroup=2
    SpawnRate=&MakeCustomFloatFromBool(Bool=$User.bIsFrontendPreview, True Float=15, False Float=20)
    Use Spawn Probability=false
   ) {module=SpawnRate002}
  ]
  particleSpawnScript [
   &VC_PaintedVariantsOverride() {enabled=false}
   /SetVariables(Particles.VisibilityTag=&SpawnGroup(Max Spawn Group=2)) {isSetParameters=true}
   &InitializeParticle(
    Color Mode=::NewEnumerator0
    Lifetime=&MakeCustomFloatFromBool(
     Bool=$User.bIsFrontendPreview
     True Float=&RandomRangeFloat(
      Minimum=&SelectFloatFromArray(
       Direct Array Index=&SpawnGroup()
       Float Selection Array=DataInterface() {propertyValues {FloatData [1.1,0.8,0.45]}}
      )
      Maximum=&SelectFloatFromArray(
       Direct Array Index=&SpawnGroup()
       Float Selection Array=DataInterface() {propertyValues {FloatData [1.1,1,0.6]}}
      )
     )
     False Float=&RandomRangeFloat(
      Minimum=&SelectFloatFromArray(
       Direct Array Index=&SpawnGroup()
       Float Selection Array=DataInterface() {propertyValues {FloatData [0.9,0.7,0.45]}}
      )
      Maximum=&SelectFloatFromArray(
       Direct Array Index=&SpawnGroup()
       Float Selection Array=DataInterface() {propertyValues {FloatData [0.9,0.7,0.6]}}
      )
     )
    )
    Sprite Rotation Mode=::NewEnumerator2
    UsePositionOffset=false
    Write Material Random=true
   )
   /SetVariables(
    Particles.Triangles_S1_Size:Vector2f=&Multiply_Vector2D(
     A=&Vector2DFromFloat(Value=&RandomRangeFloat(Minimum=12, Maximum=18))
     B=&Vector2DFromFloat(Value=3.2)
    )
    Particles.Triangles_S2_Size:Vector2f=&Multiply_Vector2D(
     A=&Vector2DFromFloat(Value=&RandomRangeFloat(Minimum=8, Maximum=18))
     B=&Vector2DFromFloat(Value=3.4)
    )
    Particles.Triangles_S3_Size:Vector2f=&Multiply_Vector2D(
     A=&Vector2DFromFloat(Value=&RandomRangeFloat(Minimum=5, Maximum=16))
     B=&Vector2DFromFloat(Value=3.8)
    )
   ) {isSetParameters=true}
   &PartitionParticles(Number Of Partitions=$Emitter.PortCount)
   &SkeletalMeshLocation(
    Bone / Socket Index=$Particles.PartitionParticles.Partition
    Bone / Socket Sampling Mode=::NewEnumerator5
    MaskSpawnGroup=false
    Mesh Sampling Type=::NewEnumerator6
    Sampled Position Offset=&AddVectorToPosition(Position=<-20 0 0>, Vector=$User.DTCarOffset)
    Skeletal Mesh=$System.SkeletalMesh
   )
   &AddVelocity(Velocity=&MaskVectorBySpawnGroup(Masked Vector=<-500 0 0>, Passthrough Vector=<0 0 0>))
   &AddVelocity(Velocity=&MaskVectorBySpawnGroup(Masked Vector=<-700 0 0>, Passthrough Vector=<0 0 0>, Spawn Group Mask=1)) {module=AddVelocity001}
   &AddVelocity(Velocity=&MaskVectorBySpawnGroup(
     Masked Vector=&MakeCustomVectorFromBool(
      Bool=$User.bIsFrontendPreview
      True Vector=&RandomRangeVector(Minimum=<-120 -30 -60>, Maximum=<-160 30 60>)
      False Vector=&RandomRangeVector(Minimum=<-90 -10 -30>, Maximum=<-130 10 30>)
     )
     Passthrough Vector=<0 0 0>
     Spawn Group Mask=2
    )) {module=AddVelocity002}
   &Color(Color=&MakeCustomLinearColorFromBool(Bool=$User.bIsPainted, True Color=$User.PaintedColor, False Color=<rgba 0.051 0 0.5451 1>))
   &ShapeLocation(
    MaskSpawnGroup=true
    Override Seed=false
    Spawn Group Mask:NiagaraInt32=&SelectIntFromArray(
     Direct Array Index=&SpawnGroup(Min Spawn Group=1, Max Spawn Group=2)
     Int Selection Array=DataInterface() {propertyValues {IntData [1,2]}}
    )
    Sphere Radius=&RandomRangeFloat(
     Minimum=8
     Maximum=&SelectFloatFromArray(
      Direct Array Index=&SpawnGroup(Min Spawn Group=1, Max Spawn Group=2)
      Float Selection Array=DataInterface() {propertyValues {FloatData [2,3]}}
     )
    )
    UseOwnerScale=true
   )
  ]
  particleUpdateScript [
   &ParticleState()
   &ScaleSpriteSize(
    Scale Sprite Size Mode=::NewEnumerator0
    Uniform Scale Factor=&MaskFloatBySpawnGroup(
     Masked Float=&FloatFromCurve(FloatCurve=DataInterface() {propertyValues {Curve {keys [<at 0.6 1>,<at 1 0>]}}})
     Passthrough Float=0
     Spawn Group Mask=2
    )
   )
   &SpriteRotationRate(Rotation Rate=&RandomRangeFloat(Minimum=-0.1, Maximum=0.1))
   &SolveForcesAndVelocity(Clamp Velocity=false, Limit Acceleration=false)
   &ScaleColor(
    Scale Alpha=&FloatFromCurve(FloatCurve=DataInterface() {propertyValues {Curve {keys [<at 0.3 1>,<at 0.9 0>]}}})
    ScaleA=true
    ScaleRGB=false
   )
   &ScaleColor(
    Scale Alpha=&MaskFloatBySpawnGroup(
     Masked Float=&FloatFromCurve(FloatCurve=DataInterface() {propertyValues {Curve {keys [
          {interpMode=RCIM_Cubic, leaveTangentWeight=0.3333, tangentMode=RCTM_User, time=0, value=0}
          {
           arriveTangent=0.0752
           arriveTangentWeight=0.0333
           interpMode=RCIM_Cubic
           leaveTangent=0.0752
           leaveTangentWeight=0.3
           tangentMode=RCTM_User
           time=0.1
           value=0.45
          }
          {
           arriveTangent=0.1859
           arriveTangentWeight=0.1
           interpMode=RCIM_Cubic
           leaveTangent=0.1859
           leaveTangentWeight=0.2
           tangentMode=RCTM_User
           time=0.4
           value=1
          }
          {
           arriveTangent=-0.2393
           arriveTangentWeight=0.1667
           interpMode=RCIM_Cubic
           leaveTangent=-0.2393
           leaveTangentWeight=0.0333
           tangentMode=RCTM_User
           time=0.9
           value=0
          }
          {arriveTangentWeight=0.0333, interpMode=RCIM_Cubic, tangentMode=RCTM_User, time=1, value=0}
         ]}}})
     Passthrough Float=&FloatFromCurve(FloatCurve=DataInterface() {propertyValues {Curve {keys [
          {interpMode=RCIM_Cubic, leaveTangentWeight=0.3333, tangentMode=RCTM_User, time=0, value=0}
          {
           arriveTangent=0.0752
           arriveTangentWeight=0.0333
           interpMode=RCIM_Cubic
           leaveTangent=0.0752
           leaveTangentWeight=0.3
           tangentMode=RCTM_User
           time=0.1
           value=0.45
          }
          {
           arriveTangent=0.1859
           arriveTangentWeight=0.1
           interpMode=RCIM_Cubic
           leaveTangent=0.1859
           leaveTangentWeight=0.2
           tangentMode=RCTM_User
           time=0.4
           value=1
          }
          {
           arriveTangent=-0.2393
           arriveTangentWeight=0.1667
           interpMode=RCIM_Cubic
           leaveTangent=-0.2393
           leaveTangentWeight=0.0333
           tangentMode=RCTM_User
           time=0.9
           value=0
          }
          {arriveTangentWeight=0.0333, interpMode=RCIM_Cubic, tangentMode=RCTM_User, time=1, value=0}
         ]}}})
     Spawn Group Mask=2
    )
    ScaleA=true
    ScaleRGB=false
   ) {module=ScaleColor001}
   &SubUVAnimation(
    Output Alternate Renderer Binding=true
    "Sprite Renderer "=DataInterface() {propertyValues {SpriteRenderer=&self:NE_VC_Boost_Sprite_1.NiagaraSpriteRendererProperties_2}}
    UseEndFrame=false
    UseStartFrame=false
   )
   &DynamicMaterialParameters(
    Index 0 Param 1=&RandomRangeFloat(Minimum=0, Maximum=1)
    Index 0 Param 2=&RandomRangeFloat(Minimum=0, Maximum=1)
    Index 0 Param 3=5
    Index 0 Param 4=0.5
    Module.Index0Param0Write=true
    Module.Index0Param1Write=true
    Module.Index0Param2Write=true
    Module.Index0Param3Write=true
   )
   /SetVariables(
    Particles.Triangles_S2_Dynamic:Vector4f=&MakeVector4(
     X=&RandomRangeFloat(Minimum=0, Maximum=1)
     Y=&RandomRangeFloat(Minimum=0, Maximum=0.5)
     Z=2
     W=0.5
    )
    Particles.Triangles_S3_Dynamic:Vector4f=&MakeVector4(
     X=&FloatFromCurve(FloatCurve=DataInterface() {propertyValues {Curve {keys [<at 0 1 RCIM_Cubic>,<at 0.8 0 RCIM_Cubic>]}}})
     Y=&RandomRangeFloat(Minimum=0, Maximum=0.1)
     Z=10
     W=1
    )
   ) {isSetParameters=true}
  ]
  rendererClasses [&NiagaraSpriteRendererProperties]
  renderers [
   {rendererClass=&NiagaraSpriteRendererProperties}
   {rendererClass=&NiagaraSpriteRendererProperties, rendererIndex=1}
   {rendererClass=&NiagaraSpriteRendererProperties, rendererIndex=2}
  ]
 }
 NE_VC_Ribbon_Parent001 {
  emitterSpawnScript [
   /SetVariables(Emitter.Painted:NiagaraBool=&Boolean_NotOperation(Bool=false)) {enabled=false, isSetParameters=true}
   /SetVariables(Emitter.No Painted:NiagaraBool=&Boolean_NotOperation(Bool=$User.bIsPainted)) {enabled=false, isSetParameters=true}
  ]
  emitterUpdateScript [
   &EmitterState()
   &VC_FilteredSocketCount()
   &VC_RateQualityHandler()
   &VC_SpawnRibbon_v1(Number of Ribbons=$Emitter.PortCount, SpawnRate=30)
  ]
  particleSpawnScript [
   &VC_PaintedVariantsOverride(
    Black=<rgba 10 10 10 2>
    BurntSienna=<rgba 100 30.8752 12.3264 1>
    Cobalt=<rgba 9.0831 0 100 1>
    Crimson=<rgba 100 15.2143 0 1>
    DefaultColor=&MakeCustomLinearColorFromBool(Bool=$User.bIsPainted, True Color=$User.PaintedColor, False Color=<rgba 0 100 99.9948 2>)
    ForestGreen=<rgba 0 88.1045 100 1>
    Lime=<rgba 9.3272 100 0 1>
    Orange=<rgba 100 56.9707 0 1>
    Pink=<rgba 100 24.6546 3.4599 1>
    Purple=<rgba 57.714 14.0625 100 1>
    Saffron=<rgba 100 81.2847 19.7917 1>
    SkyBlue=<rgba 17.7083 72.4641 100 1>
    TitaniumWhite=<rgba 150 150 150 1>
   )
   &VC_InitializeRibbon_v1(
    Color=$Particles.Color
    Custom Offset=true
    Enable Skeletal Position=true
    Lifetime=1.25
    Local Vector Offset=&AddVectorToPosition(Position=<0 0 0>, Vector=$User.DTCarOffset)
    Offset Amount=1
    Ribbon Width=70
   )
   &AddVelocity(Velocity=<-1700 0 0>) {module=AddVelocity001}
  ]
  particleUpdateScript [
   &ParticleState()
   &InheritSourceMovement(
    Applied Position Scale=&Multiply_VectorByFloat(
     Vector=<1 1 1>
     Float=&FloatFromCurve(FloatCurve=DataInterface() {propertyValues {Curve {keys [
          {arriveTangent=0.0201, interpMode=RCIM_Cubic, leaveTangent=0.0201, tangentMode=RCTM_User, time=0, value=1}
          {arriveTangentWeight=0.07, interpMode=RCIM_Cubic, leaveTangentWeight=0.2633, tangentMode=RCTM_User, time=0.2, value=1}
          {arriveTangent=0.0759, arriveTangentWeight=0.2633, interpMode=RCIM_Cubic, leaveTangent=0.0759, tangentMode=RCTM_User, time=1, value=0}
         ]}}})
    )
    ApplyForce=false
    ApplyPosition=true
    ApplyPositionFalloff=false
   )
   &SolveForcesAndVelocity(Clamp Velocity=false, Limit Acceleration=false)
   &VC_UpdateRibbon_v1(Enable Skeletal Position=true)
   &DynamicMaterialParameters(
    Index 0 Param 2=1
    Index 0 Param 3=&Waveform("Waveform [1]"=::NewEnumerator8, "[1] Frequency"=15, "[1] Amplitude Min/Max"=<0 4>)
    Index 0 Param 4=3
    Module.Index0Float0Write=false
    Module.Index0Float1Write=true
    Module.Index0Float2Write=true
    Module.Index0Float3Write=true
   )
   &ScaleColor(
    Scale Alpha=&Multiply_Float(
     A=&FloatFromCurve(FloatCurve=DataInterface() {propertyValues {Curve {
         defaultValue=1
         keys [
          {arriveTangent=9.0725, interpMode=RCIM_Cubic, leaveTangent=9.0725, tangentMode=RCTM_User, time=0, value=0}
          {arriveTangent=-2.5249, interpMode=RCIM_Cubic, leaveTangent=-2.5249, tangentMode=RCTM_User, time=0.1408, value=0.9847}
          {arriveTangent=-0.0405, interpMode=RCIM_Cubic, leaveTangent=-0.0405, tangentMode=RCTM_User, time=0.8, value=0}
         ]
        }}})
     B=&Waveform("Waveform [1]"=::NewEnumerator8, "[1] Frequency"=5, "[1] Amplitude Min/Max"=<-0.5 1>)
    )
    ScaleA=true
    ScaleRGB=false
   )
   &ScaleRibbonWidth(Ribbon Width Scale=&Waveform("Waveform [1]"=::NewEnumerator8, "[1] Frequency"=10, "[1] Amplitude Min/Max"=<0.1 1.25>))
  ]
  rendererClasses [&NiagaraRibbonRendererProperties]
  renderers [
   {rendererClass=&NiagaraRibbonRendererProperties}
   {rendererClass=&NiagaraRibbonRendererProperties, rendererIndex=1}
  ]
 }
}
summary {
 systemName=NS_Boost_Retrotech_Opt
 emitters [
  {emitterName=NE_VC_Boost_Ribbon, rendererClasses [&NiagaraRibbonRendererProperties]}
  {emitterName=NE_VC_Boost_Sprite, simTarget=GPUComputeSim, rendererClasses [&NiagaraSpriteRendererProperties]}
  {emitterName=NE_VC_Ribbon_Parent001, rendererClasses [&NiagaraRibbonRendererProperties]}
 ]
 userVariables [
  User.bIsFrontendPreview {type=NiagaraBool, defaultValue=true}
  User.bIsPainted {type=NiagaraBool}
  User.ColorVariantID {type=NiagaraInt32}
  User.DTCarOffset {type=Vector3f, description="", defaultValue {struct=Vector3f, value=<0 0 0>}}
  User.PaintedColor {type=LinearColor, description="", defaultValue {struct=LinearColor, value=<rgba 1 1 1 1>}}
  User.PaintedScalar {type=NiagaraFloat, defaultValue=1, description=""}
  User.PortCount {type=NiagaraInt32, defaultValue=2}
 ]
}
