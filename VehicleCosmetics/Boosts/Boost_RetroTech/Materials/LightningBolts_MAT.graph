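# Additive unlit Niagara ribbon and sprite material for animated lightning bolts in the RetroTech boost effect.
# A time-driven flipbook with tiled, panned UVs is sampled normally and inverted, with weighted RGB channels.
# Zero-tiled, panned triangle textures act as time-varying flicker masks that blend the two samples, controlled by dynamic parameters.
# Particle color, U-axis fades and depth fade shape the result, with emissive passing through MF_FX_EmissiveAdjustments.
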Material /VehicleCosmetics/Boosts/Boost_RetroTech/Materials/LightningBolts_MAT
graph {
 properties {
  bUsedWithNiagaraRibbons=true
  bUsedWithNiagaraSprites=true
  blendMode=BLEND_Additive
  pixelDepthOffsetMode=PDOM_Legacy
  shadingModel=MSM_Unlit
  usedShadingModels=Unlit
 }
 nodes [
  Constant() {r=0.1}
  Divide()
  n1 = ParticleColor()
  n2 = TextureCoordinate()
  n3 = Multiply(A=OneMinus(None=ComponentMask(None=@n2) {r=true})) {constB=8}
  n4 = ComponentMask(None=TextureCoordinate() {vTiling=0}) {r=true}
  n5 = ScalarParameter() {defaultValue=1, parameterName=PatternTileU}
  n6 = ScalarParameter() {parameterName=Pattern_PanU}
  n7 = ComponentMask(None=Panner(
    Coordinate=Multiply(A=@n4, B=@n5)
    Time=MaterialFunctionCall(Speed=@n6) {materialFunction=&TimeWithSpeedVariable}.Time
   ) {speedX=0.5}) {r=true}
  n8 = ComponentMask(None=TextureCoordinate() {uTiling=0}) {g=true}
  n9 = ScalarParameter() {defaultValue=1, parameterName=PatternTileV}
  n10 = ScalarParameter() {parameterName=Pattern_PanV}
  n11 = ComponentMask(None=Panner(
    Coordinate=Multiply(A=@n8, B=@n9)
    Time=MaterialFunctionCall(Speed=@n10) {materialFunction=&TimeWithSpeedVariable}.Time
   ) {speedY=0.5}) {g=true}
  n12 = ScalarParameter() {defaultValue=2, parameterName=Columns}
  n13 = Divide(B=@n12) {constB=1}
  n14 = ScalarParameter() {defaultValue=4, parameterName=Rows}
  n15 = Divide(B=@n14) {constB=1}
  n16 = Multiply(A=Time(), B=ScalarParameter() {defaultValue=16, parameterName=FB_Speed})
  n17 = Fmod(A=Floor(None=@n16), B=@n14)
  n18 = Floor(None=Divide(A=@n16, B=@n14))
  n19 = Add(
   A=Multiply(A=AppendVector(A=@n7, B=@n11), B=AppendVector(A=@n13, B=@n15))
   B=AppendVector(A=Divide(A=@n17, B=@n14), B=Divide(A=@n18, B=@n12))
  )
  n20 = TextureSample(UVs=@n19) {texture=/DealershipSharedLibrary/FX/Textures/TexturePackage/T_VCVFX_LightningFB01_RGB}
  n21 = NamedRerouteUsage()
  n22 = NamedRerouteUsage()
  n23 = Add(A=Multiply(A=NamedRerouteUsage(), B=@n20.R), B=Add(A=Multiply(A=@n21, B=@n20.G), B=Multiply(A=@n22, B=@n20.B)))
  n24 = TextureSample(UVs=OneMinus(None=@n19)) {texture=&T_VCVFX_LightningFB01_RGB}
  n25 = NamedRerouteUsage()
  n26 = NamedRerouteUsage()
  n27 = Add(A=Multiply(A=NamedRerouteUsage(), B=@n24.R), B=Add(A=Multiply(A=@n25, B=@n24.G), B=Multiply(A=@n26, B=@n24.B)))
  n28 = TextureCoordinate() {uTiling=0, vTiling=0}
  n29 = DynamicParameter() {paramNames [Brightness,FlickerSpeed,Param3,Param4]}
  n30 = Multiply(A=@n29.FlickerSpeed, B=Time())
  n31 = TextureSample(UVs=Panner(Coordinate=@n28, Time=@n30) {speedX=1, speedY=0.65}) {texture=/DealershipSharedLibrary/Assets/Sets/RetroTech/T_VCVFX_Boost_RetroTech_TrianglesFB02_RGB}
  n32 = TextureSample(UVs=Panner(Coordinate=@n28, Time=@n30) {speedX=-1, speedY=-0.3}) {texture=&T_VCVFX_Boost_RetroTech_TrianglesFB02_RGB}
  n33 = Subtract(A=AppendVector(A=@n31.R, B=@n32.R)) {constB=0.2}
  n34 = Add(A=@n31.G, B=@n32.G)
  n35 = Saturate(None=AppendVector(A=Multiply(A=@n33, B=@n29.Param4), B=Multiply(A=@n29.Param3, B=@n34)))
  n36 = ComponentMask(None=@n35) {r=true}
  n37 = ComponentMask(None=@n35) {g=true}
  n38 = ComponentMask(None=@n35) {b=true}
  n39 = Saturate(None=LinearInterpolate(
    A=NamedRerouteDeclaration(None=@n36) {name=Mask R, nodeColor=<rgba 1 0 0.7031 1>}
    B=NamedRerouteDeclaration(None=@n37) {name=Mask G, nodeColor=<rgba 0 1 0 1>}
    Alpha=NamedRerouteDeclaration(None=@n38) {name=Mask B, nodeColor=<rgba 0 0 1 1>}
   ))
  n40 = ComponentMask(None=TextureCoordinate()) {r=true}
  n41 = Multiply(A=Multiply(A=LinearInterpolate(A=@n23, B=@n27, Alpha=@n39), B=NamedRerouteUsage()), B=OneMinus(None=@n40))
  MaterialFunctionCall(Emissive=Multiply(A=@n1.RGB, B=Multiply(A=@n3, B=@n41))) {materialFunction=&MF_FX_EmissiveAdjustments}
  n42 = Multiply(A=@n41, B=Saturate(None=ConstantBiasScale(None=@n40) {bias=-0.025, scale=4}))
  Multiply(A=Saturate(None=Multiply(A=@n1.A, B=@n42)), B=DepthFade() {fadeDistanceDefault=255})
  NamedRerouteDeclaration(None=@n29.Brightness) {name=DynBrightness, nodeColor=<rgba 1 0.2578 0 1>}
 ]
}
