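# Translucent unlit Niagara ribbon material for the trail of the RetroTech vehicle boost effect.
# A panning trail texture, with switchable tiling, is desaturated and tinted by particle color for emissive, with Dynamic Parameters driving highlights and brightness.
# Its blue channel and fades at both ribbon ends shape opacity, while red feeds a distortion signal.
# Switchable world position offset adds animated sine/cosine wobble, ramped in along the ribbon's length.
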Material /VehicleCosmetics/Boosts/Boost_RetroTech/Materials/Trail_01_MAT
graph {
 properties {bUsedWithNiagaraRibbons=true, blendMode=BLEND_Translucent, pixelDepthOffsetMode=PDOM_Legacy, shadingModel=MSM_Unlit, usedShadingModels=Unlit}
 nodes [
  n1 = DynamicParameter() {paramNames [Highlights,Brightness,Distortion,InnerDetail]}
  n2 = TextureCoordinate() {uTiling=2}
  n3 = TextureCoordinate() {uTiling=1.5}
  n4 = TextureSample(UVs=StaticSwitchParameter(True=Panner(Coordinate=@n2) {speedX=-2}, False=Panner(Coordinate=@n3) {speedX=-1.5}) {parameterName=Tile2x}) {texture=/VehicleCosmetics/Boosts/Boost_RetroTech/Textures/T_VCVFX_Boost_RetroTech_TrailSimple_RGB}
  n5 = Desaturation(None=Add(A=Multiply(A=@n1.Highlights, B=@n4.R), B=NamedRerouteDeclaration(None=@n4.RGB) {name=RGB MainTex, nodeColor=<rgba 0.1016 0 1 1>}))
  n6 = ParticleColor()
  MaterialFunctionCall(Emissive=Multiply(A=Multiply(A=@n5, B=@n1.Brightness), B=@n6.RGB)) {materialFunction=&MF_FX_EmissiveAdjustments}
  NamedRerouteDeclaration(None=Add(A=NamedRerouteUsage(), B=Constant3Vector() {constant=<rgba 0 0 1 0>})) {name=Normal, nodeColor=<rgba 0 0 1 1>}
  n7 = TextureCoordinate()
  n8 = Saturate(None=ConstantBiasScale(None=ComponentMask(None=@n7) {r=true}) {bias=-0.01, scale=2})
  n9 = Time()
  n10 = TextureCoordinate() {uTiling=3}
  n11 = ScalarParameter() {defaultValue=1, parameterName=tile}
  n12 = Add(
   A=Multiply(A=ScalarParameter() {defaultValue=1, parameterName=Speed}, B=OneMinus(None=@n9))
   B=ComponentMask(None=Multiply(A=@n10, B=@n11)) {r=true}
  )
  n13 = Add(A=Cosine(None=@n12), B=Cosine(None=@n12) {period=2})
  n14 = Sine(None=@n12)
  n15 = Sine(None=@n12) {period=2}
  n16 = AppendVector(A=AppendVector(A=@n13, B=Add(A=@n14, B=@n15)), B=@n13)
  n17 = ScalarParameter() {parameterName=Amount}
  n18 = Transform(None=Multiply(A=Multiply(A=@n16, B=@n17), B=ScalarParameter() {defaultValue=1, parameterName=Intensity}))
  NamedRerouteDeclaration(None=StaticSwitchParameter(True=Multiply(A=@n8, B=@n18), False=Constant()) {defaultValue=true, parameterName=WPO}) {name=WPO, nodeColor=<rgba 1 0.2002 0 1>}
  NamedRerouteDeclaration(None=@n1.Distortion) {name=DYNDistortion, nodeColor=<rgba 0.1719 0 1 1>}
  n19 = Add(A=@n4.R, B=@n4.R)
  n20 = TextureCoordinate()
  n21 = ComponentMask(None=@n20) {r=true}
  n22 = ConstantBiasScale(None=OneMinus(None=@n21)) {bias=-0.01, scale=1}
  n23 = ConstantBiasScale(None=@n21) {bias=-0.025, scale=20}
  n24 = OneMinus(None=ComponentMask(None=@n20) {r=true})
  n25 = Multiply(A=Multiply(A=Saturate(None=@n22), B=Saturate(None=@n23)), B=Multiply(A=Power(Base=@n24)))
  NamedRerouteDeclaration(None=Multiply(A=Multiply(A=@n19, B=@n25), B=NamedRerouteUsage())) {name=Distortion, nodeColor=<rgba 1 0.1141 0.1041 1>}
  NamedRerouteUsage()
  NamedRerouteUsage()
  NamedRerouteUsage()
  n26 = Saturate(None=ConstantBiasScale(None=Multiply(A=@n4.B) {constB=8}) {bias=-0.01, scale=1})
  Saturate(None=Multiply(A=@n6.A, B=Multiply(A=@n26, B=@n25)))
 ]
}
