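# Translucent unlit Niagara sprite material drawing triangle particles with an RGB-split parallax effect for the RetroTech vehicle boost.
# A base texture sample takes particle color, while three BumpOffset copies at increasing depths are tinted red, green and blue, then added into adjusted emissive.
# Opacity sums the samples, fades UV edges (per 4x4 cell with FlipBook) and multiplies particle alpha.
# A dynamic parameter exposes texture switch, opacity and brightness controls.
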Material /VehicleCosmetics/Boosts/Boost_RetroTech/Materials/TrianglesBase_MAT
graph {
 properties {bUsedWithNiagaraSprites=true, blendMode=BLEND_Translucent, pixelDepthOffsetMode=PDOM_Legacy, shadingModel=MSM_Unlit, usedShadingModels=Unlit}
 nodes [
  n1 = Constant() {r=0.5}
  n2 = ParticleColor()
  n3 = TextureCoordinate()
  n4 = TextureSampleParameter2D(UVs=BumpOffset(Coordinate=@n3, Height=Constant() {r=-1})) {parameterName=Texture, texture=/VehicleCosmetics/Boosts/Boost_RetroTech/Textures/T_VCVFX_Boost_RetroTech_Triangle_RGB}
  n5 = NamedRerouteUsage()
  n6 = LinearInterpolate(A=@n4.R, B=@n4.R, Alpha=@n5)
  n7 = Multiply(A=NamedRerouteUsage()) {constB=0.05}
  n8 = Multiply(A=LinearInterpolate(A=@n4.R, B=@n4.R, Alpha=@n5), B=@n7)
  n9 = TextureSampleParameter2D(UVs=BumpOffset(Height=Constant() {r=-1.5})) {parameterName=Texture, texture=&T_VCVFX_Boost_RetroTech_Triangle_RGB}
  n10 = NamedRerouteUsage()
  n11 = LinearInterpolate(A=@n9.R, B=@n9.R, Alpha=@n10)
  n12 = Multiply(A=LinearInterpolate(A=@n9.R, B=@n9.R, Alpha=@n10), B=@n7)
  n13 = Add(
   A=Multiply(A=Add(A=@n6, B=@n8), B=Constant3Vector() {constant=<rgba 1 0 0 0>})
   B=Multiply(A=Add(A=@n11, B=@n12), B=Constant3Vector() {constant=<rgba 0 1 0 0>})
  )
  n14 = TextureSampleParameter2D(UVs=BumpOffset(Height=Constant() {r=-2})) {parameterName=Texture, texture=&T_VCVFX_Boost_RetroTech_Triangle_RGB}
  n15 = LinearInterpolate(A=@n14.R, B=@n14.R, Alpha=@n10)
  n16 = LinearInterpolate(A=@n14.R, B=@n14.R, Alpha=@n10)
  n17 = Multiply(A=Add(A=@n15, B=Multiply(A=@n16, B=@n7)), B=Constant3Vector() {constant=<rgba 0 0 1 0>})
  n18 = Saturate(None=Multiply(A=NamedRerouteUsage(), B=Add(A=@n13, B=@n17)))
  n19 = LinearInterpolate(A=Multiply(A=Desaturation(None=@n2.RGB, Fraction=@n1), B=@n18), B=@n18, Alpha=NamedRerouteUsage())
  n20 = TextureSampleParameter2D(UVs=@n3) {parameterName=Texture, texture=&T_VCVFX_Boost_RetroTech_Triangle_RGB}
  n21 = LinearInterpolate(A=@n20.R, B=@n20.R, Alpha=@n5)
  n22 = Multiply(A=LinearInterpolate(A=@n20.R, B=@n20.R, Alpha=@n5), B=NamedRerouteUsage())
  n23 = Add(A=Multiply(A=@n2.RGB, B=@n21), B=Multiply(A=@n22, B=@n2.RGB))
  MaterialFunctionCall(Emissive=Multiply(A=Add(A=@n23, B=@n19), B=NamedRerouteUsage())) {materialFunction=&MF_FX_EmissiveAdjustments}
  n24 = Add(A=Add(A=@n21, B=@n22), B=Add(A=Add(A=@n6, B=@n11), B=@n15))
  n25 = StaticSwitchParameter(True=Frac(None=TextureCoordinate() {uTiling=4, vTiling=4}), False=TextureCoordinate()) {parameterName=FlipBook}
  n26 = ComponentMask(None=@n25) {r=true}
  n27 = ConstantBiasScale(None=OneMinus(None=@n26)) {bias=-0.025, scale=20}
  n28 = ConstantBiasScale(None=@n26) {bias=-0.025, scale=20}
  n29 = ComponentMask(None=@n25) {g=true}
  n30 = ConstantBiasScale(None=OneMinus(None=@n29)) {bias=-0.025, scale=20}
  n31 = ConstantBiasScale(None=@n29) {bias=-0.025, scale=20}
  n32 = Multiply(A=Multiply(A=Saturate(None=@n27), B=Saturate(None=@n28)), B=Multiply(A=Saturate(None=@n30), B=Saturate(None=@n31)))
  Multiply(A=Saturate(None=Multiply(A=@n24, B=@n32)), B=@n2.A)
  n33 = DynamicParameter() {parameterIndex=1, defaultValue=<rgba 0 1 1 1>, paramNames [TextureSwitch, Opacity, Brightness, RGB_Brightness]}
  NamedRerouteDeclaration(None=@n33.TextureSwitch) {name=Dyn TextureSwitch, nodeColor=<rgba 1 0 0.4687 1>}
  NamedRerouteDeclaration(None=@n33.Opacity) {name=Dyn Opacity, nodeColor=<rgba 0 0.7891 1 1>}
  NamedRerouteDeclaration(None=@n33.Brightness) {name=DYN_Brightness, nodeColor=<rgba 1 0.8906 0 1>}
  NamedRerouteDeclaration(None=@n33.RGB_Brightness) {name=DYN_RGB_Brightness, nodeColor=<rgba 0.5703 0 1 1>}
  Reroute(None=@n33.Brightness)
  Reroute(None=@n33.RGB_Brightness)
 ]
}
